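# Physics asset for the accessory "Bits" mesh of the Dumpling Man Veggie player character, covering two hip jars, a hip pouch and a rolling pin.
# Each accessory bone has a capsule body constrained directly to a small pelvis capsule, so none chain to one another.
# Both jar constraints lock twist, while the pouch and rolling pin constraints leave it unlocked with a tighter twist angle setting.
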PhysicsAsset /BRCosmetics/Characters/Player/Male/Medium/Bodies/M_MED_Dumpling_Man_Veggie/Meshes/M_MED_Dumpling_Man_Veggie_Bits_Physics
bodies {
 dyn_hip_jar_01 [{shapeType=Capsule, length=6, radius=5, center=<3 0 0>, rotation=<rot 90 0 0>}]
 dyn_hip_jar_02 [{shapeType=Capsule, length=6, radius=5.75, center=<6.6452 0 0>, rotation=<rot 90 0 0>}]
 dyn_hip_pouch [{shapeType=Capsule, length=5.75, radius=6.4396, center=<-0.1955 0 -3.3373>}]
 dyn_rollingpin [{shapeType=Capsule, length=24.7454, radius=5.6008, center=<0 1 0>, rotation=<rot 0.226 -90 -0.456>}]
 pelvis [{shapeType=Capsule, length=1, radius=1}]
}
constraints [
 {bone1Name=dyn_hip_jar_01, bone2Name=pelvis, swing1LimitDegrees=45, swing2LimitDegrees=45, twistLimitDegrees=45, twistMotion=Locked}
 {bone1Name=dyn_hip_jar_02, bone2Name=pelvis, swing1LimitDegrees=45, swing2LimitDegrees=45, twistLimitDegrees=45, twistMotion=Locked}
 {bone1Name=dyn_rollingpin, bone2Name=pelvis, swing1LimitDegrees=45, swing2LimitDegrees=45, twistLimitDegrees=10}
 {bone1Name=dyn_hip_pouch, bone2Name=pelvis, swing1LimitDegrees=45, swing2LimitDegrees=45, twistLimitDegrees=10}
]
