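# Physics asset for the front skirt of the male medium Dumpling Man Veggie player body.
# Capsules cover pelvis, thighs, calves and feet, joined by angularly free constraints.
# Five four-box skirt chains use limited swing and tight twist, the center hanging from the pelvis and the others from the thighs.
# Angularly locked cross-links join neighboring chains at the third segment, and the left edge root alone has a much wider swing limit.
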PhysicsAsset /BRCosmetics/Characters/Player/Male/Medium/Bodies/M_MED_Dumpling_Man_Veggie/Meshes/M_MED_Dumpling_Man_Veggie_Skirt_Front_Physics
bodies {
 calf_l [{shapeType=Capsule, length=29, radius=10.5, center=<-15 0 0>, rotation=<rot 90 0 -180>}]
 calf_r [{shapeType=Capsule, length=29, radius=10.5, center=<15 0 0>, rotation=<rot 90 0 0>}]
 dyn_skirt_front_01 [{shapeType=Box, extentX=12, extentY=3.5, extentZ=15, center=<5 0 0>}]
 dyn_skirt_front_01_l [{shapeType=Box, extentX=12, extentY=3.5, extentZ=15, center=<5 0 0>}]
 dyn_skirt_front_01_r [{shapeType=Box, extentX=12, extentY=3.5, extentZ=15, center=<5 0 0>}]
 dyn_skirt_front_02 [{shapeType=Box, extentX=12, extentY=3.5, extentZ=15, center=<5 0 0>}]
 dyn_skirt_front_02_l [{shapeType=Box, extentX=12, extentY=3.5, extentZ=15, center=<5 0 0>}]
 dyn_skirt_front_02_r [{shapeType=Box, extentX=12, extentY=3.5, extentZ=15, center=<5 0 0>}]
 dyn_skirt_front_03 [{shapeType=Box, extentX=12, extentY=3.5, extentZ=15, center=<5 0 0>}]
 dyn_skirt_front_03_l [{shapeType=Box, extentX=12, extentY=3.5, extentZ=15, center=<5 0 0>}]
 dyn_skirt_front_03_r [{shapeType=Box, extentX=12, extentY=3.5, extentZ=15, center=<5 0 0>}]
 dyn_skirt_front_04 [{shapeType=Box, extentX=12, extentY=3.5, extentZ=15, center=<5 0 0>}]
 dyn_skirt_front_04_l [{shapeType=Box, extentX=12, extentY=3.5, extentZ=15, center=<5 0 0>}]
 dyn_skirt_front_04_r [{shapeType=Box, extentX=12, extentY=3.5, extentZ=15, center=<5 0 0>}]
 dyn_skirt_front_edge_01_l [{shapeType=Box, extentX=12, extentY=3.5, extentZ=15, center=<5 0 0>}]
 dyn_skirt_front_edge_01_r [{shapeType=Box, extentX=12, extentY=3.5, extentZ=15, center=<5 0 0>}]
 dyn_skirt_front_edge_02_l [{shapeType=Box, extentX=12, extentY=3.5, extentZ=15, center=<5 0 0>}]
 dyn_skirt_front_edge_02_r [{shapeType=Box, extentX=12, extentY=3.5, extentZ=15, center=<5 0 0>}]
 dyn_skirt_front_edge_03_l [{shapeType=Box, extentX=12, extentY=3.5, extentZ=15, center=<5 0 0>}]
 dyn_skirt_front_edge_03_r [{shapeType=Box, extentX=12, extentY=3.5, extentZ=15, center=<5 0 0>}]
 dyn_skirt_front_edge_04_l [{shapeType=Box, extentX=12, extentY=3.5, extentZ=15, center=<5 0 0>}]
 dyn_skirt_front_edge_04_r [{shapeType=Box, extentX=12, extentY=3.5, extentZ=15, center=<5 0 0>}]
 foot_l [{shapeType=Capsule, length=15.25, radius=10, center=<-1.0605 6.8172 0>, rotation=<rot 0 -180 -90.0001>}]
 foot_r [{shapeType=Capsule, length=15.25, radius=10, center=<1.0605 -6.8172 0>, rotation=<rot 0 0 -90.0001>}]
 pelvis [{shapeType=Capsule, length=3.7114, radius=10.6358}]
 thigh_l [{shapeType=Capsule, length=34.25, radius=12, center=<-21.747 0 -0.0001>, rotation=<rot 90 0 180>}]
 thigh_r [{shapeType=Capsule, length=34.25, radius=12, center=<21.747 0 -0.0001>, rotation=<rot 90 0 0>}]
}
constraints [
 {
  bone1Name=calf_r
  bone2Name=thigh_r
  swing1LimitDegrees=45
  swing1Motion=Free
  swing2LimitDegrees=45
  swing2Motion=Free
  twistLimitDegrees=45
  twistMotion=Free
 }
 {
  bone1Name=foot_r
  bone2Name=calf_r
  swing1LimitDegrees=45
  swing1Motion=Free
  swing2LimitDegrees=45
  swing2Motion=Free
  twistLimitDegrees=45
  twistMotion=Free
 }
 {
  bone1Name=foot_l
  bone2Name=calf_l
  swing1LimitDegrees=45
  swing1Motion=Free
  swing2LimitDegrees=45
  swing2Motion=Free
  twistLimitDegrees=45
  twistMotion=Free
 }
 {
  bone1Name=calf_l
  bone2Name=thigh_l
  swing1LimitDegrees=45
  swing1Motion=Free
  swing2LimitDegrees=45
  swing2Motion=Free
  twistLimitDegrees=45
  twistMotion=Free
 }
 {
  bone1Name=thigh_l
  bone2Name=pelvis
  swing1LimitDegrees=45
  swing1Motion=Free
  swing2LimitDegrees=45
  swing2Motion=Free
  twistLimitDegrees=45
  twistMotion=Free
 }
 {
  bone1Name=thigh_r
  bone2Name=pelvis
  swing1LimitDegrees=45
  swing1Motion=Free
  swing2LimitDegrees=45
  swing2Motion=Free
  twistLimitDegrees=45
  twistMotion=Free
 }
 {bone1Name=dyn_skirt_front_01_l, bone2Name=thigh_l, swing1LimitDegrees=20, swing2LimitDegrees=20, twistLimitDegrees=5}
 {bone1Name=dyn_skirt_front_02_l, bone2Name=dyn_skirt_front_01_l, swing1LimitDegrees=20, swing2LimitDegrees=20, twistLimitDegrees=5}
 {bone1Name=dyn_skirt_front_03_l, bone2Name=dyn_skirt_front_02_l, swing1LimitDegrees=20, swing2LimitDegrees=20, twistLimitDegrees=5}
 {bone1Name=dyn_skirt_front_04_l, bone2Name=dyn_skirt_front_03_l, swing1LimitDegrees=20, swing2LimitDegrees=20, twistLimitDegrees=5}
 {bone1Name=dyn_skirt_front_edge_01_l, bone2Name=thigh_l, swing1LimitDegrees=90, swing2LimitDegrees=20, twistLimitDegrees=5}
 {bone1Name=dyn_skirt_front_edge_02_l, bone2Name=dyn_skirt_front_edge_01_l, swing1LimitDegrees=20, swing2LimitDegrees=20, twistLimitDegrees=5}
 {bone1Name=dyn_skirt_front_edge_03_l, bone2Name=dyn_skirt_front_edge_02_l, swing1LimitDegrees=20, swing2LimitDegrees=20, twistLimitDegrees=5}
 {bone1Name=dyn_skirt_front_edge_04_l, bone2Name=dyn_skirt_front_edge_03_l, swing1LimitDegrees=20, swing2LimitDegrees=20, twistLimitDegrees=5}
 {bone1Name=dyn_skirt_front_01_r, bone2Name=thigh_r, swing1LimitDegrees=20, swing2LimitDegrees=20, twistLimitDegrees=5}
 {bone1Name=dyn_skirt_front_02_r, bone2Name=dyn_skirt_front_01_r, swing1LimitDegrees=20, swing2LimitDegrees=20, twistLimitDegrees=5}
 {bone1Name=dyn_skirt_front_03_r, bone2Name=dyn_skirt_front_02_r, swing1LimitDegrees=20, swing2LimitDegrees=20, twistLimitDegrees=5}
 {bone1Name=dyn_skirt_front_04_r, bone2Name=dyn_skirt_front_03_r, swing1LimitDegrees=20, swing2LimitDegrees=20, twistLimitDegrees=5}
 {bone1Name=dyn_skirt_front_edge_01_r, bone2Name=thigh_r, swing1LimitDegrees=20, swing2LimitDegrees=20, twistLimitDegrees=5}
 {bone1Name=dyn_skirt_front_edge_02_r, bone2Name=dyn_skirt_front_edge_01_r, swing1LimitDegrees=20, swing2LimitDegrees=20, twistLimitDegrees=5}
 {bone1Name=dyn_skirt_front_edge_03_r, bone2Name=dyn_skirt_front_edge_02_r, swing1LimitDegrees=20, swing2LimitDegrees=20, twistLimitDegrees=5}
 {bone1Name=dyn_skirt_front_edge_04_r, bone2Name=dyn_skirt_front_edge_03_r, swing1LimitDegrees=20, swing2LimitDegrees=20, twistLimitDegrees=5}
 {bone1Name=dyn_skirt_front_01, bone2Name=pelvis, swing1LimitDegrees=20, swing2LimitDegrees=20, twistLimitDegrees=5}
 {bone1Name=dyn_skirt_front_02, bone2Name=dyn_skirt_front_01, swing1LimitDegrees=20, swing2LimitDegrees=20, twistLimitDegrees=5}
 {bone1Name=dyn_skirt_front_03, bone2Name=dyn_skirt_front_02, swing1LimitDegrees=20, swing2LimitDegrees=20, twistLimitDegrees=5}
 {bone1Name=dyn_skirt_front_04, bone2Name=dyn_skirt_front_03, swing1LimitDegrees=20, swing2LimitDegrees=20, twistLimitDegrees=5}
 {
  bone1Name=dyn_skirt_front_03_l
  bone2Name=dyn_skirt_front_edge_03_l
  swing1LimitDegrees=45
  swing1Motion=Locked
  swing2LimitDegrees=45
  swing2Motion=Locked
  twistLimitDegrees=45
  twistMotion=Locked
 }
 {
  bone1Name=dyn_skirt_front_03_r
  bone2Name=dyn_skirt_front_edge_03_r
  swing1LimitDegrees=45
  swing1Motion=Locked
  swing2LimitDegrees=45
  swing2Motion=Locked
  twistLimitDegrees=45
  twistMotion=Locked
 }
 {
  bone1Name=dyn_skirt_front_03_r
  bone2Name=dyn_skirt_front_03
  swing1LimitDegrees=45
  swing1Motion=Locked
  swing2LimitDegrees=45
  swing2Motion=Locked
  twistLimitDegrees=45
  twistMotion=Locked
 }
 {
  bone1Name=dyn_skirt_front_03_l
  bone2Name=dyn_skirt_front_03
  swing1LimitDegrees=45
  swing1Motion=Locked
  swing2LimitDegrees=45
  swing2Motion=Locked
  twistLimitDegrees=45
  twistMotion=Locked
 }
]
