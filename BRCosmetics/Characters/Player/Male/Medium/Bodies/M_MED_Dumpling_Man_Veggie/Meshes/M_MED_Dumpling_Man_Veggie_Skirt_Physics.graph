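# Physics asset for the skirt on the Dumpling Man Veggie player body cosmetic.
# Five chains of five box bodies (back, left/right mid, left/right side) attach to the pelvis with free swing and twist, using tight angular limits between segments and thicker lower boxes.
# Angularly free cross-constraints tie mid to side chains at the fourth segment on each side.
# Pelvis, thigh, calf and foot capsules, joined by angularly free constraints, provide lower-body shapes.
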PhysicsAsset /BRCosmetics/Characters/Player/Male/Medium/Bodies/M_MED_Dumpling_Man_Veggie/Meshes/M_MED_Dumpling_Man_Veggie_Skirt_Physics
bodies {
 calf_l [{shapeType=Capsule, length=27.4338, radius=12.0662, center=<-15 -5.0264 -0.0001>, rotation=<rot 90 0 -179.9999>}]
 calf_r [{shapeType=Capsule, length=27.4338, radius=12.0662, center=<15 2.2359 0.0003>, rotation=<rot 90 83.2785 83.2786>}]
 dyn_skirt_back_01 [{shapeType=Box, extentX=15, extentY=2, extentZ=15, center=<7 0 0>, rotation=<rot 90 0 0>}]
 dyn_skirt_back_02 [{shapeType=Box, extentX=15, extentY=2, extentZ=15, center=<7 0 0>, rotation=<rot 90 0 0>}]
 dyn_skirt_back_03 [{shapeType=Box, extentX=15, extentY=2, extentZ=15, center=<7 0 0>, rotation=<rot 90 0 0>}]
 dyn_skirt_back_04 [{shapeType=Box, extentX=15, extentY=5, extentZ=15, center=<7 0 0>, rotation=<rot 90 0 0>}]
 dyn_skirt_back_05 [{shapeType=Box, extentX=15, extentY=5, extentZ=15, center=<7 0 0>, rotation=<rot 90 0 0>}]
 dyn_skirt_mid_01_l [{shapeType=Box, extentX=15, extentY=2, extentZ=15, center=<7 0 0>, rotation=<rot 90 0 0>}]
 dyn_skirt_mid_01_r [{shapeType=Box, extentX=15, extentY=2, extentZ=15, center=<7 0 0>, rotation=<rot 90 0 0>}]
 dyn_skirt_mid_02_l [{shapeType=Box, extentX=15, extentY=2, extentZ=15, center=<7 0 0>, rotation=<rot 90 0 0>}]
 dyn_skirt_mid_02_r [{shapeType=Box, extentX=15, extentY=2, extentZ=15, center=<7 0 0>, rotation=<rot 90 0 0>}]
 dyn_skirt_mid_03_l [{shapeType=Box, extentX=15, extentY=2, extentZ=15, center=<7 0 0>, rotation=<rot 90 0 0>}]
 dyn_skirt_mid_03_r [{shapeType=Box, extentX=15, extentY=2, extentZ=15, center=<7 0 0>, rotation=<rot 90 0 0>}]
 dyn_skirt_mid_04_l [{shapeType=Box, extentX=15, extentY=5, extentZ=15, center=<7 0 0>, rotation=<rot 90 0 0>}]
 dyn_skirt_mid_04_r [{shapeType=Box, extentX=15, extentY=5, extentZ=15, center=<7 0 0>, rotation=<rot 90 0 0>}]
 dyn_skirt_mid_05_l [{shapeType=Box, extentX=15, extentY=5, extentZ=15, center=<7 0 0>, rotation=<rot 90 0 0>}]
 dyn_skirt_mid_05_r [{shapeType=Box, extentX=15, extentY=5, extentZ=15, center=<7 0 0>, rotation=<rot 90 0 0>}]
 dyn_skirt_side_01_l [{shapeType=Box, extentX=15, extentY=2, extentZ=15, center=<7 0 0>, rotation=<rot 90 0 0>}]
 dyn_skirt_side_01_r [{shapeType=Box, extentX=15, extentY=2, extentZ=15, center=<7 0 0>, rotation=<rot 90 0 0>}]
 dyn_skirt_side_02_l [{shapeType=Box, extentX=15, extentY=2, extentZ=15, center=<7 0 0>, rotation=<rot 90 0 0>}]
 dyn_skirt_side_02_r [{shapeType=Box, extentX=15, extentY=2, extentZ=15, center=<7 0 0>, rotation=<rot 90 0 0>}]
 dyn_skirt_side_03_l [{shapeType=Box, extentX=15, extentY=2, extentZ=15, center=<7 0 0>, rotation=<rot 90 0 0>}]
 dyn_skirt_side_03_r [{shapeType=Box, extentX=15, extentY=2, extentZ=15, center=<7 0 0>, rotation=<rot 90 0 0>}]
 dyn_skirt_side_04_l [{shapeType=Box, extentX=15, extentY=5, extentZ=15, center=<7 0 0>, rotation=<rot 90 0 0>}]
 dyn_skirt_side_04_r [{shapeType=Box, extentX=15, extentY=5, extentZ=15, center=<7 0 0>, rotation=<rot 90 0 0>}]
 dyn_skirt_side_05_l [{shapeType=Box, extentX=15, extentY=5, extentZ=15, center=<7 0 0>, rotation=<rot 90 0 0>}]
 dyn_skirt_side_05_r [{shapeType=Box, extentX=15, extentY=5, extentZ=15, center=<7 0 0>, rotation=<rot 90 0 0>}]
 foot_l [{shapeType=Capsule, length=15.25, radius=10, center=<-1.0605 6.8172 0>, rotation=<rot 0 -180 -90.0001>}]
 foot_r [{shapeType=Capsule, length=15.25, radius=10, center=<1.0605 -6.8172 0>, rotation=<rot 0 0 -90.0001>}]
 pelvis [{shapeType=Capsule, length=3.7114, radius=10.6358}]
 thigh_l [{shapeType=Capsule, length=35.5, radius=10.75, center=<-21.747 0 -0.0001>, rotation=<rot 90 0 180>}]
 thigh_r [{shapeType=Capsule, length=35.5, radius=10.75, center=<21.747 0 -0.0001>, rotation=<rot 90 0 0>}]
}
constraints [
 {bone1Name=dyn_skirt_mid_02_r, bone2Name=dyn_skirt_mid_01_r, swing1LimitDegrees=20, swing2LimitDegrees=20, twistLimitDegrees=5}
 {bone1Name=dyn_skirt_mid_03_r, bone2Name=dyn_skirt_mid_02_r, swing1LimitDegrees=20, swing2LimitDegrees=20, twistLimitDegrees=5}
 {bone1Name=dyn_skirt_mid_04_r, bone2Name=dyn_skirt_mid_03_r, swing1LimitDegrees=20, swing2LimitDegrees=20, twistLimitDegrees=5}
 {bone1Name=dyn_skirt_mid_05_r, bone2Name=dyn_skirt_mid_04_r, swing1LimitDegrees=20, swing2LimitDegrees=20, twistLimitDegrees=5}
 {bone1Name=dyn_skirt_side_02_r, bone2Name=dyn_skirt_side_01_r, swing1LimitDegrees=20, swing2LimitDegrees=20, twistLimitDegrees=5}
 {bone1Name=dyn_skirt_side_03_r, bone2Name=dyn_skirt_side_02_r, swing1LimitDegrees=20, swing2LimitDegrees=20, twistLimitDegrees=5}
 {bone1Name=dyn_skirt_side_04_r, bone2Name=dyn_skirt_side_03_r, swing1LimitDegrees=20, swing2LimitDegrees=20, twistLimitDegrees=5}
 {bone1Name=dyn_skirt_side_05_r, bone2Name=dyn_skirt_side_04_r, swing1LimitDegrees=20, swing2LimitDegrees=20, twistLimitDegrees=5}
 {bone1Name=dyn_skirt_back_02, bone2Name=dyn_skirt_back_01, swing1LimitDegrees=20, swing2LimitDegrees=20, twistLimitDegrees=5}
 {bone1Name=dyn_skirt_back_03, bone2Name=dyn_skirt_back_02, swing1LimitDegrees=20, swing2LimitDegrees=20, twistLimitDegrees=5}
 {bone1Name=dyn_skirt_back_04, bone2Name=dyn_skirt_back_03, swing1LimitDegrees=20, swing2LimitDegrees=20, twistLimitDegrees=5}
 {bone1Name=dyn_skirt_back_05, bone2Name=dyn_skirt_back_04, swing1LimitDegrees=20, swing2LimitDegrees=20, twistLimitDegrees=5}
 {bone1Name=dyn_skirt_mid_02_l, bone2Name=dyn_skirt_mid_01_l, swing1LimitDegrees=20, swing2LimitDegrees=20, twistLimitDegrees=5}
 {bone1Name=dyn_skirt_mid_03_l, bone2Name=dyn_skirt_mid_02_l, swing1LimitDegrees=20, swing2LimitDegrees=20, twistLimitDegrees=5}
 {bone1Name=dyn_skirt_mid_04_l, bone2Name=dyn_skirt_mid_03_l, swing1LimitDegrees=20, swing2LimitDegrees=20, twistLimitDegrees=5}
 {bone1Name=dyn_skirt_mid_05_l, bone2Name=dyn_skirt_mid_04_l, swing1LimitDegrees=20, swing2LimitDegrees=20, twistLimitDegrees=5}
 {bone1Name=dyn_skirt_side_05_l, bone2Name=dyn_skirt_side_04_l, swing1LimitDegrees=20, swing2LimitDegrees=20, twistLimitDegrees=5}
 {bone1Name=dyn_skirt_side_02_l, bone2Name=dyn_skirt_side_01_l, swing1LimitDegrees=20, swing2LimitDegrees=20, twistLimitDegrees=5}
 {bone1Name=dyn_skirt_side_03_l, bone2Name=dyn_skirt_side_02_l, swing1LimitDegrees=20, swing2LimitDegrees=20, twistLimitDegrees=5}
 {bone1Name=dyn_skirt_side_04_l, bone2Name=dyn_skirt_side_03_l, swing1LimitDegrees=20, swing2LimitDegrees=20, twistLimitDegrees=5}
 {
  bone1Name=calf_r
  bone2Name=thigh_r
  swing1LimitDegrees=45
  swing1Motion=Free
  swing2LimitDegrees=45
  swing2Motion=Free
  twistLimitDegrees=45
  twistMotion=Free
 }
 {
  bone1Name=foot_r
  bone2Name=calf_r
  swing1LimitDegrees=45
  swing1Motion=Free
  swing2LimitDegrees=45
  swing2Motion=Free
  twistLimitDegrees=45
  twistMotion=Free
 }
 {
  bone1Name=foot_l
  bone2Name=calf_l
  swing1LimitDegrees=45
  swing1Motion=Free
  swing2LimitDegrees=45
  swing2Motion=Free
  twistLimitDegrees=45
  twistMotion=Free
 }
 {
  bone1Name=calf_l
  bone2Name=thigh_l
  swing1LimitDegrees=45
  swing1Motion=Free
  swing2LimitDegrees=45
  swing2Motion=Free
  twistLimitDegrees=45
  twistMotion=Free
 }
 {
  bone1Name=dyn_skirt_mid_04_l
  bone2Name=dyn_skirt_side_04_l
  swing1LimitDegrees=45
  swing1Motion=Free
  swing2LimitDegrees=45
  swing2Motion=Free
  twistLimitDegrees=45
  twistMotion=Free
 }
 {
  bone1Name=dyn_skirt_mid_04_r
  bone2Name=dyn_skirt_side_04_r
  swing1LimitDegrees=45
  swing1Motion=Free
  swing2LimitDegrees=45
  swing2Motion=Free
  twistLimitDegrees=45
  twistMotion=Free
 }
 {
  bone1Name=thigh_l
  bone2Name=pelvis
  swing1LimitDegrees=45
  swing1Motion=Free
  swing2LimitDegrees=45
  swing2Motion=Free
  twistLimitDegrees=45
  twistMotion=Free
 }
 {
  bone1Name=thigh_r
  bone2Name=pelvis
  swing1LimitDegrees=45
  swing1Motion=Free
  swing2LimitDegrees=45
  swing2Motion=Free
  twistLimitDegrees=45
  twistMotion=Free
 }
 {
  bone1Name=dyn_skirt_mid_01_r
  bone2Name=pelvis
  swing1LimitDegrees=45
  swing1Motion=Free
  swing2LimitDegrees=45
  swing2Motion=Free
  twistLimitDegrees=45
  twistMotion=Free
 }
 {
  bone1Name=dyn_skirt_side_01_r
  bone2Name=pelvis
  swing1LimitDegrees=45
  swing1Motion=Free
  swing2LimitDegrees=45
  swing2Motion=Free
  twistLimitDegrees=45
  twistMotion=Free
 }
 {
  bone1Name=dyn_skirt_back_01
  bone2Name=pelvis
  swing1LimitDegrees=45
  swing1Motion=Free
  swing2LimitDegrees=45
  swing2Motion=Free
  twistLimitDegrees=45
  twistMotion=Free
 }
 {
  bone1Name=dyn_skirt_mid_01_l
  bone2Name=pelvis
  swing1LimitDegrees=45
  swing1Motion=Free
  swing2LimitDegrees=45
  swing2Motion=Free
  twistLimitDegrees=45
  twistMotion=Free
 }
 {
  bone1Name=dyn_skirt_side_01_l
  bone2Name=pelvis
  swing1LimitDegrees=45
  swing1Motion=Free
  swing2LimitDegrees=45
  swing2Motion=Free
  twistLimitDegrees=45
  twistMotion=Free
 }
]
