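# Anim Blueprint for the Dumpling Man Veggie body, adding sleeve correctives and skirt, prop and pouch motion to the parent's copied pose.
# Full path layers crouch Control Rig, AnimDynamics and RigidBody, with skydiving force on skirts.
# LOD above 1 or disabled rigid bodies select cheaper constraint, look-at, trail and spring nodes.
# Crouch state and left-thigh yaw (despite "right_thigh" naming) drive offsets; a front-skirt AnimDynamics chain is unconnected.
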AnimBlueprint /BRCosmetics/Characters/Player/Male/Medium/Bodies/M_MED_Dumpling_Man_Veggie/Meshes/M_MED_Dumpling_Man_Veggie_AnimBP
target_skeleton=&/M_MED_Dumpling_Man_Veggie_Skeleton
graphs [
 AnimGraph {nodes [
   Bone Driven Controller!(ComponentPose:Component Space Pose Link Structure, Alpha:float=1) -> (Pose:Component Space Pose Link Structure) {is=title AnimGraphNode_BoneDrivenController, alphaBoolBlend {blendOption=Linear}}
   n1 = Get crouch_right_thigh_delta_alpha() -> (crouch_right_thigh_delta_alpha:double)
   n2 = "Transform (Modify) Bone - Bone: dyn_skirt_front_01_l"(Alpha=@n1, Translation="0,0,0", Rotation="0,25,30", Scale="1,1,1") {rotationMode=BMM_Additive, alphaBoolBlend {blendOption=Linear}, boneToModify {boneName=dyn_skirt_front_01_l}, rotation {roll=30, yaw=25}}
   n3 = "Transform (Modify) Bone - Bone: dyn_skirt_front_01_r"(ComponentPose=@n2, Alpha=@n1, Translation="0,0,0", Rotation="10.000,0.000,-5", Scale="1,1,1") {rotationMode=BMM_Additive, alphaBoolBlend {blendOption=Linear}, boneToModify {boneName=dyn_skirt_front_01_r}, rotation {pitch=10, roll=-5}}
   n4 = "Transform (Modify) Bone - Bone: dyn_skirt_front_edge_01_l"(ComponentPose=@n3, Alpha=@n1, Translation="0,0,0", Rotation="0,25,0", Scale="1,1,1") {rotationMode=BMM_Additive, alphaBoolBlend {blendOption=Linear}, boneToModify {boneName=dyn_skirt_front_edge_01_l}, rotation {yaw=25}}
   n5 = "Transform (Modify) Bone - Bone: dyn_skirt_front_edge_01_r"(ComponentPose=@n4, Alpha=@n1, Translation="0,0,0", Rotation="9.962,-0.880,-5.077", Scale="1,1,1") {
    rotationMode=BMM_Additive
    alphaBoolBlend {blendOption=Linear}
    boneToModify {boneName=dyn_skirt_front_edge_01_r}
    rotation=<rot 9.962 -0.88 -5.0771>
   }
   n6 = Anim Dynamics - Bone: dyn_skirt_front_01_r(ComponentPose=@n5, Alpha=1, ExternalForce="0,0,0") {
    bGravityOverrideInSimSpace=true
    gravityScale=0.35
    simulationSpace=RootRelative
    alphaBoolBlend {blendOption=Linear}
    boundBone {boneName=dyn_skirt_front_01_r}
    physicsBodyDefinitions [{
      sphereCollisionRadius=0
      boundBone {boneName=dyn_skirt_front_01_r}
      boxExtents=<5 30 30>
      constraintSetup {angularLimitsMax=<0 0 30>, angularLimitsMin=<0 0 0>, angularTarget=<0 0 0>}
      localJointOffset=<5 0 0>
     }]
   }
   n7 = Anim Dynamics - Bone: dyn_skirt_front_02_r(ComponentPose=@n6, Alpha=1, ExternalForce="0,0,0") {
    bGravityOverrideInSimSpace=true
    gravityScale=0.35
    simulationSpace=RootRelative
    alphaBoolBlend {blendOption=Linear}
    boundBone {boneName=dyn_skirt_front_02_r}
    physicsBodyDefinitions [{
      sphereCollisionRadius=0
      boundBone {boneName=dyn_skirt_front_02_r}
      boxExtents=<5 30 30>
      constraintSetup {angularLimitsMax=<0 0 30>, angularLimitsMin=<0 0 0>, angularTarget=<0 0 0>}
      localJointOffset=<5 0 0>
     }]
   }
   n8 = Anim Dynamics - Bone: dyn_skirt_front_03_r(ComponentPose=@n7, Alpha=1, ExternalForce="0,0,0") {
    bGravityOverrideInSimSpace=true
    gravityScale=0.35
    simulationSpace=RootRelative
    alphaBoolBlend {blendOption=Linear}
    boundBone {boneName=dyn_skirt_front_03_r}
    physicsBodyDefinitions [{
      sphereCollisionRadius=0
      boundBone {boneName=dyn_skirt_front_03_r}
      boxExtents=<5 30 30>
      constraintSetup {angularLimitsMax=<0 5 30>, angularLimitsMin=<0 -5 0>, angularTarget=<0 0 0>}
      localJointOffset=<5 0 0>
     }]
   }
   n9 = Anim Dynamics - Bone: dyn_skirt_front_04_r(ComponentPose=@n8, Alpha=1, ExternalForce="0,0,0") {
    bGravityOverrideInSimSpace=true
    gravityScale=0.35
    simulationSpace=RootRelative
    alphaBoolBlend {blendOption=Linear}
    boundBone {boneName=dyn_skirt_front_04_r}
    physicsBodyDefinitions [{
      sphereCollisionRadius=0
      boundBone {boneName=dyn_skirt_front_04_r}
      boxExtents=<5 30 30>
      constraintSetup {angularLimitsMax=<0 5 30>, angularLimitsMin=<0 -5 0>, angularTarget=<0 0 0>}
      localJointOffset=<5 0 0>
     }]
   }
   n10 = Anim Dynamics - Bone: dyn_skirt_front_01_l(ComponentPose=@n9, Alpha=1, ExternalForce="0,0,0") {
    bGravityOverrideInSimSpace=true
    gravityScale=0.35
    simulationSpace=RootRelative
    alphaBoolBlend {blendOption=Linear}
    boundBone {boneName=dyn_skirt_front_01_l}
    physicsBodyDefinitions [{
      sphereCollisionRadius=0
      boundBone {boneName=dyn_skirt_front_01_l}
      boxExtents=<5 30 30>
      constraintSetup {angularLimitsMax=<0 0 30>, angularLimitsMin=<0 0 0>, angularTarget=<0 0 0>}
      localJointOffset=<5 0 0>
     }]
   }
   n11 = Anim Dynamics - Bone: dyn_skirt_front_02_l(ComponentPose=@n10, Alpha=1, ExternalForce="0,0,0") {
    bGravityOverrideInSimSpace=true
    gravityScale=0.35
    simulationSpace=RootRelative
    alphaBoolBlend {blendOption=Linear}
    boundBone {boneName=dyn_skirt_front_02_l}
    physicsBodyDefinitions [{
      sphereCollisionRadius=0
      boundBone {boneName=dyn_skirt_front_02_l}
      boxExtents=<5 30 30>
      constraintSetup {angularLimitsMax=<0 0 30>, angularLimitsMin=<0 0 0>, angularTarget=<0 0 0>}
      localJointOffset=<5 0 0>
     }]
   }
   n12 = Anim Dynamics - Bone: dyn_skirt_front_03_l(ComponentPose=@n11, Alpha=1, ExternalForce="0,0,0") {
    bGravityOverrideInSimSpace=true
    gravityScale=0.35
    simulationSpace=RootRelative
    alphaBoolBlend {blendOption=Linear}
    boundBone {boneName=dyn_skirt_front_03_l}
    physicsBodyDefinitions [{
      sphereCollisionRadius=0
      boundBone {boneName=dyn_skirt_front_03_l}
      boxExtents=<5 30 30>
      constraintSetup {angularLimitsMax=<0 5 30>, angularLimitsMin=<0 -5 0>, angularTarget=<0 0 0>}
      localJointOffset=<5 0 0>
     }]
   }
   n13 = Anim Dynamics - Bone: dyn_skirt_front_04_l(ComponentPose=@n12, Alpha=1, ExternalForce="0,0,0") {
    bGravityOverrideInSimSpace=true
    gravityScale=0.35
    simulationSpace=RootRelative
    alphaBoolBlend {blendOption=Linear}
    boundBone {boneName=dyn_skirt_front_04_l}
    physicsBodyDefinitions [{
      sphereCollisionRadius=0
      boundBone {boneName=dyn_skirt_front_04_l}
      boxExtents=<5 30 30>
      constraintSetup {angularLimitsMax=<0 5 30>, angularLimitsMin=<0 -5 0>, angularTarget=<0 0 0>}
      localJointOffset=<5 0 0>
     }]
   }
   n14 = Anim Dynamics - Bone: dyn_skirt_front_edge_01_r(ComponentPose=@n13, Alpha=1, ExternalForce="0,0,0") {
    bGravityOverrideInSimSpace=true
    gravityScale=0.35
    simulationSpace=RootRelative
    alphaBoolBlend {blendOption=Linear}
    boundBone {boneName=dyn_skirt_front_edge_01_r}
    physicsBodyDefinitions [{
      sphereCollisionRadius=0
      boundBone {boneName=dyn_skirt_front_edge_01_r}
      boxExtents=<5 30 30>
      constraintSetup {angularLimitsMax=<0 0 30>, angularLimitsMin=<0 0 0>, angularTarget=<0 0 0>}
      localJointOffset=<5 0 0>
     }]
   }
   n15 = Anim Dynamics - Bone: dyn_skirt_front_edge_02_r(ComponentPose=@n14, Alpha=1, ExternalForce="0,0,0") {
    bGravityOverrideInSimSpace=true
    gravityScale=0.35
    simulationSpace=RootRelative
    alphaBoolBlend {blendOption=Linear}
    boundBone {boneName=dyn_skirt_front_edge_02_r}
    physicsBodyDefinitions [{
      sphereCollisionRadius=0
      boundBone {boneName=dyn_skirt_front_edge_02_r}
      boxExtents=<5 30 30>
      constraintSetup {angularLimitsMax=<0 0 30>, angularLimitsMin=<0 0 0>, angularTarget=<0 0 0>}
      localJointOffset=<5 0 0>
     }]
   }
   n16 = Anim Dynamics - Bone: dyn_skirt_front_edge_03_r(ComponentPose=@n15, Alpha=1, ExternalForce="0,0,0") {
    bGravityOverrideInSimSpace=true
    gravityScale=0.35
    simulationSpace=RootRelative
    alphaBoolBlend {blendOption=Linear}
    boundBone {boneName=dyn_skirt_front_edge_03_r}
    physicsBodyDefinitions [{
      sphereCollisionRadius=0
      boundBone {boneName=dyn_skirt_front_edge_03_r}
      boxExtents=<5 30 30>
      constraintSetup {angularLimitsMax=<0 5 30>, angularLimitsMin=<0 -5 0>, angularTarget=<0 0 0>}
      localJointOffset=<5 0 0>
     }]
   }
   n17 = Anim Dynamics - Bone: dyn_skirt_front_edge_04_r(ComponentPose=@n16, Alpha=1, ExternalForce="0,0,0") {
    bGravityOverrideInSimSpace=true
    gravityScale=0.35
    simulationSpace=RootRelative
    alphaBoolBlend {blendOption=Linear}
    boundBone {boneName=dyn_skirt_front_edge_04_r}
    physicsBodyDefinitions [{
      sphereCollisionRadius=0
      boundBone {boneName=dyn_skirt_front_edge_04_r}
      boxExtents=<5 30 30>
      constraintSetup {angularLimitsMax=<0 5 30>, angularLimitsMin=<0 -5 0>, angularTarget=<0 0 0>}
      localJointOffset=<5 0 0>
     }]
   }
   n18 = Anim Dynamics - Bone: dyn_skirt_front_edge_01_l(ComponentPose=@n17, Alpha=1, ExternalForce="0,0,0") {
    bGravityOverrideInSimSpace=true
    gravityScale=0.35
    simulationSpace=RootRelative
    alphaBoolBlend {blendOption=Linear}
    boundBone {boneName=dyn_skirt_front_edge_01_l}
    physicsBodyDefinitions [{
      sphereCollisionRadius=0
      boundBone {boneName=dyn_skirt_front_edge_01_l}
      boxExtents=<5 30 30>
      constraintSetup {angularLimitsMax=<0 0 30>, angularLimitsMin=<0 0 0>, angularTarget=<0 0 0>}
      localJointOffset=<5 0 0>
     }]
   }
   n19 = Anim Dynamics - Bone: dyn_skirt_front_edge_02_l(ComponentPose=@n18, Alpha=1, ExternalForce="0,0,0") {
    bGravityOverrideInSimSpace=true
    gravityScale=0.35
    simulationSpace=RootRelative
    alphaBoolBlend {blendOption=Linear}
    boundBone {boneName=dyn_skirt_front_edge_02_l}
    physicsBodyDefinitions [{
      sphereCollisionRadius=0
      boundBone {boneName=dyn_skirt_front_edge_02_l}
      boxExtents=<5 30 30>
      constraintSetup {angularLimitsMax=<0 0 30>, angularLimitsMin=<0 0 0>, angularTarget=<0 0 0>}
      localJointOffset=<5 0 0>
     }]
   }
   n20 = Anim Dynamics - Bone: dyn_skirt_front_edge_03_l(ComponentPose=@n19, Alpha=1, ExternalForce="0,0,0") {
    bGravityOverrideInSimSpace=true
    gravityScale=0.35
    simulationSpace=RootRelative
    alphaBoolBlend {blendOption=Linear}
    boundBone {boneName=dyn_skirt_front_edge_03_l}
    physicsBodyDefinitions [{
      sphereCollisionRadius=0
      boundBone {boneName=dyn_skirt_front_edge_03_l}
      boxExtents=<5 30 30>
      constraintSetup {angularLimitsMax=<0 5 30>, angularLimitsMin=<0 -5 0>, angularTarget=<0 0 0>}
      localJointOffset=<5 0 0>
     }]
   }
   n21 = Anim Dynamics - Bone: dyn_skirt_front_edge_04_l(ComponentPose=@n20, Alpha=1, ExternalForce="0,0,0") {
    bGravityOverrideInSimSpace=true
    gravityScale=0.35
    simulationSpace=RootRelative
    alphaBoolBlend {blendOption=Linear}
    boundBone {boneName=dyn_skirt_front_edge_04_l}
    physicsBodyDefinitions [{
      sphereCollisionRadius=0
      boundBone {boneName=dyn_skirt_front_edge_04_l}
      boxExtents=<5 30 30>
      constraintSetup {angularLimitsMax=<0 5 30>, angularLimitsMin=<0 -5 0>, angularTarget=<0 0 0>}
      localJointOffset=<5 0 0>
     }]
   }
   n22 = "Constraint - dyn_skirt_front_01 (PP)"!(ComponentPose=@n21, Alpha=1, ConstraintWeights_0=0.5, ConstraintWeights_1:float=0.5) -> (Pose) {
    alphaBoolBlend {blendOption=Linear}
    boneToModify {boneName=dyn_skirt_front_01}
    constraintSetup [{targetBone {boneName=dyn_skirt_front_01_l}}, {targetBone {boneName=dyn_skirt_front_01_r}}]
    constraintWeights [1,1]
   }
   n23 = "Constraint - dyn_skirt_front_02 (PP)"!(ComponentPose=@n22, Alpha=1, ConstraintWeights_0=0.5, ConstraintWeights_1:float=0.5) -> (Pose) {
    alphaBoolBlend {blendOption=Linear}
    boneToModify {boneName=dyn_skirt_front_02}
    constraintSetup [{targetBone {boneName=dyn_skirt_front_02_l}}, {targetBone {boneName=dyn_skirt_front_02_r}}]
    constraintWeights [1,1]
   }
   n24 = "Constraint - dyn_skirt_front_03 (PP)"!(ComponentPose=@n23, Alpha=1, ConstraintWeights_0=0.5, ConstraintWeights_1:float=0.5) -> (Pose) {
    alphaBoolBlend {blendOption=Linear}
    boneToModify {boneName=dyn_skirt_front_03}
    constraintSetup [{targetBone {boneName=dyn_skirt_front_03_l}}, {targetBone {boneName=dyn_skirt_front_03_r}}]
    constraintWeights [1,1]
   }
   "Constraint - dyn_skirt_front_04 (PP)"!(ComponentPose=@n24, Alpha=1, ConstraintWeights_0=0.5, ConstraintWeights_1:float=0.5) -> (Pose) {
    alphaBoolBlend {blendOption=Linear}
    boneToModify {boneName=dyn_skirt_front_04}
    constraintSetup [{targetBone {boneName=dyn_skirt_front_04_l}}, {targetBone {boneName=dyn_skirt_front_04_r}}]
    constraintWeights [1,1]
   }
   Look At - Bone: dyn_onions(Alpha=1) {
    alphaBoolBlend {blendOption=Linear}
    boneToModify {boneName=dyn_onions}
    lookAtLocation=<0 10>
    lookAtTarget {boneReference {boneName=dyn_skirt_front_03_r}}
    lookAt_Axis {axis=<1 0>}
   }
   n25 = "Constraint - dyn_skirt_front_01 (PP)"(ComponentPose=Local To Component(LocalPose=Use cached pose 'CopyPoseCache'()), Alpha=1, ConstraintWeights_0=0.5, ConstraintWeights_1=0.5) {
    alphaBoolBlend {blendOption=Linear}
    boneToModify {boneName=dyn_skirt_front_01}
    constraintSetup [{targetBone {boneName=dyn_skirt_front_01_l}}, {targetBone {boneName=dyn_skirt_front_01_r}}]
    constraintWeights [1,1]
   }
   n26 = "Constraint - dyn_skirt_front_02 (PP)"(ComponentPose=@n25, Alpha=1, ConstraintWeights_0=0.5, ConstraintWeights_1=0.5) {
    alphaBoolBlend {blendOption=Linear}
    boneToModify {boneName=dyn_skirt_front_02}
    constraintSetup [{targetBone {boneName=dyn_skirt_front_02_l}}, {targetBone {boneName=dyn_skirt_front_02_r}}]
    constraintWeights [1,1]
   }
   n27 = "Constraint - dyn_skirt_front_03 (PP)"(ComponentPose=@n26, Alpha=1, ConstraintWeights_0=0.5, ConstraintWeights_1=0.5) {
    alphaBoolBlend {blendOption=Linear}
    boneToModify {boneName=dyn_skirt_front_03}
    constraintSetup [{targetBone {boneName=dyn_skirt_front_03_l}}, {targetBone {boneName=dyn_skirt_front_03_r}}]
    constraintWeights [1,1]
   }
   n28 = "Constraint - dyn_skirt_front_04 (PP)"(ComponentPose=@n27, Alpha=1, ConstraintWeights_0=0.5, ConstraintWeights_1=0.5) {
    alphaBoolBlend {blendOption=Linear}
    boneToModify {boneName=dyn_skirt_front_04}
    constraintSetup [{targetBone {boneName=dyn_skirt_front_04_l}}, {targetBone {boneName=dyn_skirt_front_04_r}}]
    constraintWeights [1,1]
   }
   n29 = Look At - Bone: dyn_spoon(ComponentPose=@n28, Alpha=1) {
    alphaBoolBlend {blendOption=Linear}
    boneToModify {boneName=dyn_spoon}
    lookAtLocation=<0 -10 8>
    lookAtTarget {boneReference {boneName=dyn_skirt_front_03_r}}
    lookAt_Axis {axis=<1 0>}
   }
   n30 = Look At - Bone: dyn_onions(ComponentPose=@n29, Alpha=1) {
    alphaBoolBlend {blendOption=Linear}
    boneToModify {boneName=dyn_onions}
    lookAtLocation {x=0}
    lookAtTarget {boneReference {boneName=dyn_skirt_front_02_r}}
    lookAt_Axis {axis=<1 0>}
   }
   n31 = Trail controller - Bone: dyn_skirt_back_05(ComponentPose=@n30, Alpha=0.8) {
    chainLength=4
    maxDeltaTime=0.0333
    relaxationSpeedScale=2
    rotationLimits [{limitMax=<180 180 180>}, {limitMax=<180 180 180>}, {limitMax=<180 180 180>}, {limitMax=<180 180 180>}]
    rotationOffsets [<0 0 0>, <0 0 0>, <0 0 0>, <0 0 0>]
    trailBone {boneName=dyn_skirt_back_05}
   }
   n32 = Trail controller - Bone: dyn_skirt_mid_05_l(ComponentPose=@n31, Alpha=0.75) {
    chainLength=4
    maxDeltaTime=0.0333
    relaxationSpeedScale=2
    rotationLimits [{limitMax=<180 180 180>}, {limitMax=<180 180 180>}, {limitMax=<180 180 180>}, {limitMax=<180 180 180>}]
    rotationOffsets [<0 0 0>, <0 0 0>, <0 0 0>, <0 0 0>]
    trailBone {boneName=dyn_skirt_mid_05_l}
   }
   n33 = Trail controller - Bone: dyn_skirt_side_05_l(ComponentPose=@n32, Alpha=0.5) {
    chainLength=4
    maxDeltaTime=0.0333
    relaxationSpeedScale=2
    rotationLimits [{limitMax=<180 180 180>}, {limitMax=<180 180 180>}, {limitMax=<180 180 180>}, {limitMax=<180 180 180>}]
    rotationOffsets [<0 0 0>, <0 0 0>, <0 0 0>, <0 0 0>]
    trailBone {boneName=dyn_skirt_side_05_l}
   }
   n34 = Trail controller - Bone: dyn_skirt_mid_05_r(ComponentPose=@n33, Alpha=0.75) {
    chainLength=4
    maxDeltaTime=0.0333
    relaxationSpeedScale=2
    rotationLimits [{limitMax=<180 180 180>}, {limitMax=<180 180 180>}, {limitMax=<180 180 180>}, {limitMax=<180 180 180>}]
    rotationOffsets [<0 0 0>, <0 0 0>, <0 0 0>, <0 0 0>]
    trailBone {boneName=dyn_skirt_mid_05_r}
   }
   n35 = Trail controller - Bone: dyn_skirt_side_05_r(ComponentPose=@n34, Alpha=0.5) {
    chainLength=4
    maxDeltaTime=0.0333
    relaxationSpeedScale=2
    rotationLimits [{limitMax=<180 180 180>}, {limitMax=<180 180 180>}, {limitMax=<180 180 180>}, {limitMax=<180 180 180>}]
    rotationOffsets [<0 0 0>, <0 0 0>, <0 0 0>, <0 0 0>]
    trailBone {boneName=dyn_skirt_side_05_r}
   }
   n36 = Spring controller - Bone: dyn_hip_pouch!(ComponentPose:Component Space Pose Link Structure=@n35, Alpha:float=1) -> (Pose:Component Space Pose Link Structure) {
    is=title AnimGraphNode_SpringBone
    bLimitDisplacement=true
    maxDisplacement=10
    springDamping=10
    springStiffness=200
    alphaBoolBlend {blendOption=Linear}
    springBone {boneName=dyn_hip_pouch}
   }
   n37 = Get crouch_right_thigh_delta_alpha()
   n38 = "Transform (Modify) Bone - Bone: dyn_pouch_hip_01"(ComponentPose=@n36, Alpha=@n37, Translation="2.198,1.168,-0.825", Rotation="50.000,0.000,0.000", Scale="1,1,1") {
    rotationMode=BMM_Additive
    translationMode=BMM_Additive
    alphaBoolBlend {blendOption=Linear}
    boneToModify {boneName=dyn_pouch_hip_01}
    rotation {pitch=49.9999}
    translation=<2.198 1.168 -0.8248>
   }
   n39 = "Transform (Modify) Bone - Bone: dyn_hip_pouch_hip_01"(ComponentPose=@n38, Alpha=@n37, Translation="2.825,0.000,20", Rotation="25.000,0.000,0.000", Scale="1,1,1") {
    rotationMode=BMM_Additive
    translationMode=BMM_Additive
    alphaBoolBlend {blendOption=Linear}
    boneToModify {boneName=dyn_hip_pouch_hip_01}
    rotation {pitch=25}
    translation {x=2.825, z=20}
   }
   n40 = "Transform (Modify) Bone - Bone: dyn_hip_pouch_hip_02"(ComponentPose=@n39, Alpha=@n37, Translation="0,0,0", Rotation="0,0,0", Scale="1,1,1") {rotationMode=BMM_Additive, translationMode=BMM_Additive, alphaBoolBlend {blendOption=Linear}, boneToModify {boneName=dyn_hip_pouch_hip_02}}
   n41 = Control Rig(Source=Use cached pose 'CopyPoseCache'(), Alpha=Get crouch_alpha() -> (crouch_alpha:double)) {
    alphaBoolBlend {blendOption=Linear}
    controlRigAssetReference {blueprintRigClass=&/M_MED_Dumpling_Man_Veggie_CtrlRig.M_MED_Dumpling_Man_Veggie_CtrlRig_C}
   }
   n42 = "Constraint - dyn_onions (R)"(ComponentPose=Local To Component(LocalPose=@n41), Alpha=1, ConstraintWeights_0=1) {
    alphaBoolBlend {blendOption=Linear}
    boneToModify {boneName=dyn_onions}
    constraintSetup [{transformType=Rotation, targetBone {boneName=dyn_skirt_front_01_r}}]
    constraintWeights [1]
   }
   n43 = Get crouch_alpha()
   n44 = "Transform (Modify) Bone - Bone: dyn_onions"(ComponentPose=@n42, Alpha=@n43, Translation="0,0,0", Rotation="-10,10,0.000", Scale="1,1,1") {
    rotationMode=BMM_Additive
    translationMode=BMM_Additive
    alphaBoolBlend {blendOption=Linear}
    boneToModify {boneName=dyn_onions}
    rotation {pitch=-10, yaw=10}
   }
   n45 = Anim Dynamics - Bone: dyn_onions_01(ComponentPose=@n44, Alpha=1, ExternalForce="0,0,0") {
    angularBiasOverride=0.9999
    bGravityOverrideInSimSpace=true
    bOverrideAngularDamping=true
    gravityScale=0.35
    simulationSpace=RootRelative
    alphaBoolBlend {blendOption=Linear}
    boundBone {boneName=dyn_onions_01}
    physicsBodyDefinitions [{
      collisionType=CustomSphere
      sphereCollisionRadius=2
      boundBone {boneName=dyn_onions_01}
      boxExtents=<5 30 30>
      constraintSetup {angularLimitsMax=<0 10 10>, angularLimitsMin=<0 -30 -10>, angularTarget=<0 0 0>}
      localJointOffset=<5 0 -1.2>
     }]
    planarLimits [
     {drivingBone {boneName=dyn_skirt_front_01_r}, planeTransform {location=<0 0 0>, rotation=<rot 0 0 90>}}
     {drivingBone {boneName=dyn_skirt_front_02_r}, planeTransform {location=<0 0 0>, rotation=<rot 0 0 90>}}
    ]
   }
   n46 = Anim Dynamics - Bone: dyn_onions_02(ComponentPose=@n45, Alpha=1, ExternalForce="0,0,0") {
    angularBiasOverride=0.9999
    bGravityOverrideInSimSpace=true
    bOverrideAngularDamping=true
    gravityScale=0.35
    simulationSpace=RootRelative
    alphaBoolBlend {blendOption=Linear}
    boundBone {boneName=dyn_onions_02}
    physicsBodyDefinitions [{
      collisionType=CustomSphere
      sphereCollisionRadius=2
      boundBone {boneName=dyn_onions_02}
      boxExtents=<5 30 30>
      constraintSetup {angularLimitsMax=<0 10 10>, angularLimitsMin=<0 -30 -10>, angularTarget=<0 0 0>}
      localJointOffset=<5 0 -1.2>
     }]
    planarLimits [
     {drivingBone {boneName=dyn_skirt_front_01}, planeTransform {location=<0 0 0>, rotation=<rot 0 0 90>}}
     {drivingBone {boneName=dyn_skirt_front_02}, planeTransform {location=<0 0 0>, rotation=<rot 0 0 90>}}
    ]
   }
   n47 = Anim Dynamics - Bone: dyn_onions_03(ComponentPose=@n46, Alpha=1, ExternalForce="0,0,0") {
    angularBiasOverride=0.9999
    bGravityOverrideInSimSpace=true
    bOverrideAngularDamping=true
    gravityScale=0.35
    simulationSpace=RootRelative
    alphaBoolBlend {blendOption=Linear}
    boundBone {boneName=dyn_onions_03}
    physicsBodyDefinitions [{
      collisionType=CustomSphere
      sphereCollisionRadius=2
      boundBone {boneName=dyn_onions_03}
      boxExtents=<5 30 30>
      constraintSetup {angularLimitsMax=<0 10 10>, angularLimitsMin=<0 -30 -10>, angularTarget=<0 0 0>}
      localJointOffset=<5 0 -1.2>
     }]
    planarLimits [
     {drivingBone {boneName=dyn_skirt_front_01}, planeTransform {location=<0 0 0>, rotation=<rot 0 0 90>}}
     {drivingBone {boneName=dyn_skirt_front_02}, planeTransform {location=<0 0 0>, rotation=<rot 0 0 90>}}
    ]
   }
   n48 = Look At - Bone: dyn_spoon(ComponentPose=@n47, Alpha=1) {
    alphaBoolBlend {blendOption=Linear}
    boneToModify {boneName=dyn_spoon}
    lookAtLocation=<0 -10 8>
    lookAtTarget {boneReference {boneName=dyn_skirt_front_03_r}}
    lookAt_Axis {axis=<1 0>}
   }
   n49 = Anim Dynamics - Bone: dyn_spoon(ComponentPose=@n48, Alpha=1, ExternalForce="0,0,0") {
    angularDampingOverride=0.9999
    bGravityOverrideInSimSpace=true
    bOverrideAngularDamping=true
    gravityScale=0.35
    simulationSpace=RootRelative
    alphaBoolBlend {blendOption=Linear}
    boundBone {boneName=dyn_spoon}
    physicsBodyDefinitions [{
      collisionType=CustomSphere
      sphereCollisionRadius=2
      boundBone {boneName=dyn_spoon}
      boxExtents=<5 20 20>
      constraintSetup {angularLimitsMax=<0 0 5>, angularLimitsMin=<0 -30 -5>, angularTarget=<0 0 0>}
      localJointOffset=<10 0 5>
     }]
    planarLimits [
     {drivingBone {boneName=dyn_skirt_front_01}, planeTransform {location=<0 0 0>, rotation=<rot 0 0 90>}}
     {drivingBone {boneName=dyn_skirt_front_02}, planeTransform {location=<0 0 0>, rotation=<rot 0 0 90>}}
    ]
   }
   n50 = "Transform (Modify) Bone - Bone: dyn_spoon"(ComponentPose=@n49, Alpha=@n43, Translation="0,0,0", Rotation="-14.942,1.338,-5.175", Scale="1,1,1") {
    rotationMode=BMM_Additive
    translationMode=BMM_Additive
    alphaBoolBlend {blendOption=Linear}
    boneToModify {boneName=dyn_spoon}
    rotation=<rot -14.942 1.338 -5.1751>
   }
   n51 = vector * float(
    A=Break Fort Anim Input Skydiving External Force(FortAnimInput_SkydivingExternalForce=Get SkydivingExternalForceInput()) -> (
     bUseSkydivingVectorForce=~
     SkydivingVectorForce=~
     HeadToPelvisDirection=~
     FloatingMultiplier=~
     FloatingAdditive=~
     DivingMultiplier=~
     DivingAdditive=~
     ParachutingMultiplier=~
     ParachutingAdditive=~
    ).SkydivingVectorForce
    B=0.25
   )
   n52 = RigidBody(ComponentPose=@n50, ExternalForce=@n51) {
    bOverrideWorldGravity=true
    overridePhysicsAsset=&/M_MED_Dumpling_Man_Veggie_Skirt_Front_Physics
    simulationSpace=BaseBoneSpace
    alphaBoolBlend {blendOption=Linear}
    baseBoneRef {boneName=pelvis}
    componentAppliedLinearAccClamp=<1 1 1>
    overrideWorldGravity {z=-980}
   }
   n53 = RigidBody(ComponentPose=@n52, ExternalForce=@n51) {
    bOverrideWorldGravity=true
    overridePhysicsAsset=&/M_MED_Dumpling_Man_Veggie_Skirt_Physics
    simulationSpace=BaseBoneSpace
    alphaBoolBlend {blendOption=Linear}
    baseBoneRef {boneName=root}
    overrideWorldGravity {z=-980}
   }
   n54 = RigidBody(ComponentPose=@n53) {
    bOverrideWorldGravity=true
    overridePhysicsAsset=&/M_MED_Dumpling_Man_Veggie_Bits_Physics
    simulationSpace=BaseBoneSpace
    alphaBoolBlend {blendOption=Linear}
    baseBoneRef {boneName=root}
    overrideWorldGravity {z=-980}
   }
   n55 = Get crouch_right_thigh_delta_alpha()
   n56 = "Transform (Modify) Bone - Bone: dyn_pouch_hip_01"(ComponentPose=@n54, Alpha=@n55, Translation="2.198,1.168,-0.825", Rotation="50.000,0.000,0.000", Scale="1,1,1") {
    rotationMode=BMM_Additive
    translationMode=BMM_Additive
    alphaBoolBlend {blendOption=Linear}
    boneToModify {boneName=dyn_pouch_hip_01}
    rotation {pitch=49.9999}
    translation=<2.198 1.168 -0.8248>
   }
   n57 = "Transform (Modify) Bone - Bone: dyn_hip_pouch_hip_01"(ComponentPose=@n56, Alpha=@n55, Translation="2.825,0.000,20", Rotation="25.000,0.000,0.000", Scale="1,1,1") {
    rotationMode=BMM_Additive
    translationMode=BMM_Additive
    alphaBoolBlend {blendOption=Linear}
    boneToModify {boneName=dyn_hip_pouch_hip_01}
    rotation {pitch=25}
    translation {x=2.825, z=20}
   }
   n58 = "Transform (Modify) Bone - Bone: dyn_hip_pouch_hip_02"(ComponentPose=@n57, Alpha=@n55, Translation="0,0,0", Rotation="0,0,-15", Scale="1,1,1") {
    rotationMode=BMM_Additive
    translationMode=BMM_Additive
    alphaBoolBlend {blendOption=Linear}
    boneToModify {boneName=dyn_hip_pouch_hip_02}
    rotation {roll=-15}
   }
   n59 = "Transform (Modify) Bone - Bone: dyn_rollingpin"(ComponentPose=@n58, Alpha=@n55, Translation="0,0,0", Rotation="20,0,0", Scale="1,1,1") {
    rotationMode=BMM_Additive
    translationMode=BMM_Additive
    alphaBoolBlend {blendOption=Linear}
    boneToModify {boneName=dyn_rollingpin}
    rotation {pitch=20}
   }
   n60 = "integer > integer"(A=Get CurrentLOD(), B=1) {is=title K2Node_CallFunction}
   n61 = NOT Boolean(A=Get bRigidBodyEnabled())
   Output Pose(Result=Blend Poses by bool(
     BlendPose_0=Component To Local(ComponentPose=@n40)
     BlendPose_1=Component To Local(ComponentPose=@n59)
     bActiveValue=OR Boolean(A=@n61, B=@n60)
    ) {blendType=Linear}) {layerGroup=None}
   n62 = Pose Driver - Source: upperarm_l!(SourcePose:Pose Link Structure=Copy Pose From Mesh() {bUseAttachedParent=true}) -> (Pose:Pose Link Structure) {
    is=title AnimGraphNode_PoseDriver
    poseAsset=&/M_MED_Dumpling_Man_Veggie_RBF_PoseAsset
    onlyDriveBones [{boneName=dyn_sleeve_corrective}]
    poseTargets [
     {
      bApplyCustomCurve=false
      bIsHidden=false
      distanceMethod=DefaultMethod
      drivenName=Base
      functionType=DefaultFunction
      targetScale=1
      boneTransforms [{targetRotation=<rot -64.7854 -3.089 6.0959>, targetTranslation=<15.5111 0 0>}]
      customCurve {defaultValue=3.4028234663852886e+38, postInfinityExtrap=RCCE_Constant, preInfinityExtrap=RCCE_Constant, keys []}
      targetRotation=<rot 0 0 0>
     }
     {
      bApplyCustomCurve=false
      bIsHidden=false
      distanceMethod=DefaultMethod
      drivenName=ADS
      functionType=DefaultFunction
      targetScale=0.9611
      boneTransforms [{targetRotation=<rot -35.9468 70.6415 -73.477>, targetTranslation=<15.5111 0 0>}]
      customCurve {defaultValue=3.4028234663852886e+38, postInfinityExtrap=RCCE_Constant, preInfinityExtrap=RCCE_Constant, keys []}
      targetRotation=<rot 0 0 0>
     }
     {
      bApplyCustomCurve=false
      bIsHidden=false
      distanceMethod=DefaultMethod
      drivenName=Pose_0
      functionType=DefaultFunction
      targetScale=0.6317
      boneTransforms [{targetRotation=<rot -6.4821 27.3524 -79.5333>, targetTranslation=<15.5111 0 0>}]
      customCurve {defaultValue=3.4028234663852886e+38, postInfinityExtrap=RCCE_Constant, preInfinityExtrap=RCCE_Constant, keys []}
      targetRotation=<rot 0 0 0>
     }
     {
      bApplyCustomCurve=false
      bIsHidden=false
      distanceMethod=DefaultMethod
      drivenName=Pose_1
      functionType=DefaultFunction
      targetScale=0.6317
      boneTransforms [{targetRotation=<rot 10.8319 54.7356 -99.9349>, targetTranslation=<15.5111 0 0>}]
      customCurve {defaultValue=3.4028234663852886e+38, postInfinityExtrap=RCCE_Constant, preInfinityExtrap=RCCE_Constant, keys []}
      targetRotation=<rot 0 0 0>
     }
    ]
    rBFParams {radius=25.5589}
    sourceBones [{boneName=upperarm_l}]
   }
   Save cached pose 'CopyPoseCache'(Pose=@n62) {cacheName=CopyPoseCache}
   Flap Node - Bone: dyn_onions!(ComponentPose:Component Space Pose Link Structure, Alpha:float=1) -> (Pose:Component Space Pose Link Structure) {
    is=title FortAnimGraphNode_Flap
    bUseClamp=true
    targetClampMin=-3.14
    targetComponent=RotationZ
    alphaBoolBlend {blendOption=Linear}
    sourceDrives [{
      bUseRange=false
      drivingCurve=null
      multiplier=-1
      rangeMax=1
      rangeMin=-1
      remappedMax=1
      remappedMin=0
      sourceComponent=RotationZ
      useQuaternion=true
      sourceBone {boneName=thigh_r}
     }]
    targetBone {boneName=dyn_onions}
   }
  ]}
 EventGraph {nodes [
   Event BlueprintUpdateAnimation()
   n1 = NOT Boolean(A=Get bIsCrouchSprinting())
   n2 = Get bIsCrouching()
   > Set crouch_alpha(crouch_alpha:double=SelectFloat(A=1, B=0, bPickA=AND Boolean(A=@n1, B=@n2))) -> (Output_Get:double)
   n3 = GetPartSkeletalMeshComponent()
   n4 = BreakRotator(InRot=BreakTransform(InTransform=GetDeltaTransformFromRefPose(self=@n3, BoneName=thigh_l, BaseName=None)).Rotation)
   > Set crouch_right_thigh_delta_alpha(crouch_right_thigh_delta_alpha:double=float - float(B=MapRangeClamped(Value=@n4.Yaw, InRangeA=-102, InRangeB=-20)) {is=title K2Node_CallFunction}) -> (Output_Get:double)
   Event OnBeginCrouching()
   Event OnEndCrouching()
   Set crouch_alpha(crouch_alpha=1)
  ]}
]
variables [
 crouch_alpha {type=number}
 crouch_right_thigh_delta_alpha {type=number}
]
