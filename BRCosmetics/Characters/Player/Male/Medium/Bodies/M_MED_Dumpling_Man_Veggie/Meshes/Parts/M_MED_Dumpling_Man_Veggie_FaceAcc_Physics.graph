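# Physics asset for the face accessory mesh of the Dumpling Man Veggie medium male character cosmetic, covering two dynamic onion hat bones.
# Each onion bone has a capsule body, and a smaller capsule sits on the head bone.
# Both onion bodies are constrained directly to the head rather than chained, with moderate swing limit angles and a very narrow twist limit angle.
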PhysicsAsset /BRCosmetics/Characters/Player/Male/Medium/Bodies/M_MED_Dumpling_Man_Veggie/Meshes/Parts/M_MED_Dumpling_Man_Veggie_FaceAcc_Physics
bodies {
 dyn_hat_onion_A [{shapeType=Capsule, length=3.25, radius=3.25, center=<1.4273 0 0>, rotation=<rot 90 0 0>}]
 dyn_hat_onion_B [{shapeType=Capsule, length=3.25, radius=3.25, center=<1.4273 0 0>, rotation=<rot 90 0 0>}]
 head [{shapeType=Capsule, length=1, radius=2}]
}
constraints [
 {bone1Name=dyn_hat_onion_B, bone2Name=head, swing1LimitDegrees=45, swing2LimitDegrees=45, twistLimitDegrees=1}
 {bone1Name=dyn_hat_onion_A, bone2Name=head, swing1LimitDegrees=45, swing2LimitDegrees=45, twistLimitDegrees=1}
]
